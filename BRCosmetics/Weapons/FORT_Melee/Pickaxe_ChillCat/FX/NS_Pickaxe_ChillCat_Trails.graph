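# Niagara system for the weapon trail of the ChillCat pickaxe, a Fortnite melee cosmetic.
# A single AnimTrail emitter uses the shared anim trail spawn, initialize and update modules to build a short-lived ribbon from bones or sockets on a user-supplied skeletal mesh.
# Initial color stays white, with alpha taken from a curve over system loop age.
# Each particle then fades through an alpha curve while randomized dynamic material parameters feed the ribbon material.
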NiagaraSystem /BRCosmetics/Weapons/FORT_Melee/Pickaxe_ChillCat/FX/NS_Pickaxe_ChillCat_Trails
dependencies {
 usedDataInterfaces [&NiagaraDataInterfaceColorCurve, &NiagaraDataInterfaceCurve]
 usedDynamicInputs []
 usedModules [&SystemState]
}
emitters {AnimTrail {
  emitterUpdateScript [
   &EmitterState()
   &FN_SpawnAnimTrail_v2(Spawn Interval=0)
  ]
  particleSpawnScript [
   /Game/Effects/Niagara/Modules/AnimTrail/FN_InitializeAnimTrail/FN_InitializeAnimTrailBoneIndecies(Second Bone or Socket Index=1)
   /Game/Effects/Niagara/Modules/AnimTrail/FN_InitializeAnimTrail/FN_InitializeAnimTrail_v2(
    AnimTrailWidth=1.2
    Color:LinearColor=&ColorFromCurve(
     ColorCurve=DataInterface() {propertyValues {
       AlphaCurve {keys [<at 0.0685 0.5 RCIM_Cubic>,<at 0.1777 0 RCIM_Cubic>]}
       BlueCurve {keys [<at 0 1 RCIM_Cubic>]}
       GreenCurve {keys [<at 0 1 RCIM_Cubic>]}
       RedCurve {keys [<at 0 1 RCIM_Cubic>]}
      }}
     CurveIndex=$System.NormalizedLoopAge
    )
    Lifetime=0.6
   )
  ]
  particleUpdateScript [
   &ParticleState()
   &FN_UpdateAnimTrail_v2()
   &SolveForcesAndVelocity(Clamp Velocity=false, Limit Acceleration=false)
   &DynamicMaterialParameters(
    Index 0 Param 1=&UniformRangedFloat(Minimum=0, Maximum=1)
    Index 0 Param 2=&UniformRangedFloat(Minimum=0, Maximum=1)
    Index 0 Param 3=1
    Index 0 Param 4=1
    Module.Param0WriteEnabled=true
    Module.Param1WriteEnabled=true
    Module.Param2WriteEnabled=true
    Module.Param3WriteEnabled=true
   )
   &ScaleColor(
    Scale Alpha=&FloatFromCurve(FloatCurve=DataInterface() {propertyValues {Curve {keys [
         {arriveTangent=-5.203, interpMode=RCIM_Cubic, leaveTangent=-5.203, tangentMode=RCTM_User, time=0.3, value=1}
         {arriveTangent=-0.4707, interpMode=RCIM_Cubic, leaveTangent=-0.4707, tangentMode=RCTM_User, time=0.7, value=0.05}
         {arriveTangent=-0.0246, interpMode=RCIM_Cubic, leaveTangent=-0.0246, tangentMode=RCTM_User, time=0.9, value=0}
        ]}}})
    ScaleA=true
    ScaleRGB=true
   )
  ]
  rendererClasses [&NiagaraRibbonRendererProperties]
  renderers [{rendererClass=&NiagaraRibbonRendererProperties}]
 }}
summary {
 systemName=NS_Pickaxe_ChillCat_Trails
 emitters [{emitterName=AnimTrail, rendererClasses [&NiagaraRibbonRendererProperties]}]
 userVariables [
  User.RibbonIDOffset {type=NiagaraInt32}
  User.SkeletalMesh {
   type=NiagaraDataInterfaceSkeletalMesh
   defaultValue {
    struct=NiagaraExt_VariableValue_DataInterface
    value {dataInterface=&self:NiagaraDataInterfaceSkeletalMesh_0}
   }
  }
 ]
}
